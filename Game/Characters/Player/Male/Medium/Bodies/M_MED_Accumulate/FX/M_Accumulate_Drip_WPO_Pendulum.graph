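# Translucent subsurface material for Niagara mesh-particle drips on the Accumulate player character, adding pendulum swing and travelling bulges through World Position Offset.
# Dynamic parameters supply mesh origin, rest normal and positions to rotate vertices about the origin, masked along UV V, with PreviousFrameSwitch handling the previous frame.
# Two gradient textures panned by Drip Pan push vertices along normals, behind a static switch.
# Fresnel-blended colors with depth fade feed an emissive-adjustment function.
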Material /Game/Characters/Player/Male/Medium/Bodies/M_MED_Accumulate/FX/M_Accumulate_Drip_WPO_Pendulum
graph {
 properties {
  bUsedWithNiagaraMeshParticles=true
  blendMode=BLEND_Translucent
  pixelDepthOffsetMode=PDOM_Legacy
  shadingModel=MSM_Subsurface
  usedShadingModels=Subsurface
 }
 nodes [
  Constant()
  n1 = OneMinus(None=Fresnel(ExponentIn=ScalarParameter() {defaultValue=1.5, parameterName=Color Exponent}))
  n2 = Fresnel(ExponentIn=ScalarParameter() {defaultValue=1.5, group=Galaxy, parameterName=Fresnel Exponent}) {baseReflectFraction=0}
  n3 = LinearInterpolate(
   A=Multiply(
    A=VectorParameter() {parameterName=Base Color, defaultValue=<rgba 0 0 0 1>}.RGB
    B=ScalarParameter() {defaultValue=1, parameterName=BG_ColorBrightness}
   )
   B=LinearInterpolate(
    A=VectorParameter() {parameterName=Color_01, defaultValue=<rgba 1 1 1 1>}.RGB
    B=VectorParameter() {parameterName=Color_02, defaultValue=<rgba 0 0.4993 0.535 1>}.RGB
    Alpha=Power(Base=@n1)
   )
   Alpha=Clamp(None=Saturate(None=@n2))
  )
  n4 = ScalarParameter() {defaultValue=0.35, parameterName=Fresnel Brightness}
  n5 = ScalarParameter() {defaultValue=1, parameterName=Opacity}
  n6 = ScalarParameter() {defaultValue=2.75, parameterName=Depth Fade Distance}
  MaterialFunctionCall(Emissive=Multiply(A=Multiply(A=@n3, B=@n4), B=DepthFade(Opacity=@n5, FadeDistance=@n6))) {materialFunction=&MF_FX_EmissiveAdjustments}
  n7 = Subtract(A=AppendVector(A=Constant(), B=Constant()), B=TextureCoordinate())
  n8 = DynamicParameter() {parameterIndex=1, defaultValue=<rgba 0 0 -1 0>, paramNames [RestNormalX, RestNormalY, RestNormalZ, Drip Pan]}
  n9 = TextureSampleParameter2D(UVs=Clamp(None=Add(A=@n7, B=@n8.Drip Pan))) {
   group=WPO
   parameterName=WPO Texture
   samplerSource=SSM_Clamp_WorldGroupSettings
   texture=/Game/Global/Textures/Default/Masks_and_Gradients/gradient_stasis
  }
  n10 = Multiply(
   A=ScalarParameter() {defaultValue=1.1, parameterName=Power Multiply}
   B=Power(Base=@n9.RGB, Exp=ScalarParameter() {defaultValue=1, parameterName=Power Exponent})
  )
  n11 = Multiply(A=@n8.Drip Pan, B=Constant() {r=0.7})
  n12 = Add(A=AppendVector(A=Constant(), B=Constant() {r=0.3}), B=TextureCoordinate() {vTiling=0.8})
  n13 = TextureSampleParameter2D(UVs=Clamp(None=Subtract(A=@n11, B=@n12))) {
   group=WPO
   parameterName=WPO Texture Second
   samplerSource=SSM_Clamp_WorldGroupSettings
   texture=&gradient_stasis
  }
  n14 = Multiply(
   A=ScalarParameter() {defaultValue=1, parameterName=Power Multiply Second}
   B=Power(Base=@n13.RGB, Exp=ScalarParameter() {defaultValue=1, parameterName=Power Exponent Second})
  )
  n15 = Add(
   A=Multiply(A=ScalarParameter() {defaultValue=1.5, parameterName=WPO Intensity}, B=Clamp(None=@n10))
   B=Multiply(A=ScalarParameter() {defaultValue=1.25, parameterName=WPO Intensity Second}, B=Clamp(None=@n14))
  )
  n16 = VertexNormalWS()
  n17 = WorldPosition()
  n18 = DynamicParameter() {parameterIndex=2, paramNames [MeshOriginX,MeshOriginY,MeshOriginZ,MaskFalloff]}
  n19 = MaterialFunctionCall(R=@n18.MeshOriginX, G=@n18.MeshOriginX, B=@n18.MeshOriginX, A=@n18.MeshOriginX) {materialFunction=&AppendMany}
  n20 = DynamicParameter() {parameterIndex=3, defaultValue=<rgba 1 1 1 0.15>, paramNames [PosX,PosY,PosZ,Rigidity]}
  n21 = Normalize(VectorInput=Subtract(
    A=MaterialFunctionCall(R=@n20.PosX, G=@n20.PosX, B=@n20.PosX) {materialFunction=&AppendMany}.RGB
    B=@n19.RGB
   ))
  n22 = DynamicParameter() {parameterIndex=1, defaultValue=<rgba 0 0 -1 0>, paramNames [RestNormalX, RestNormalY, RestNormalZ, Drip Pan]}
  n23 = Multiply(A=MaterialFunctionCall(R=@n22.RestNormalX, G=@n22.RestNormalX, B=@n22.RestNormalX) {materialFunction=&AppendMany}.RGB) {constB=-1}
  n24 = Subtract(A=ComponentMask(None=TextureCoordinate()) {g=true}, B=ScalarParameter() {parameterName=Mask Offset})
  n25 = Saturate(None=Power(Base=Saturate(None=@n24), Exp=@n18.MaskFalloff))
  n26 = LinearInterpolate(
   A=RotateAboutAxis(NormalizedRotationAxis=CrossProduct(A=@n23, B=@n21), RotationAngle=@n20.Rigidity, PivotPoint=@n19.RGB, Position=@n17.XYZ)
   Alpha=@n25
  ) {constB=0}
  n27 = DynamicParameter() {paramNames [PrevPosX,PrevPosY,PrevPosZ,UseWPO]}
  n28 = Normalize(VectorInput=Subtract(
    A=MaterialFunctionCall(R=@n27.PrevPosX, G=@n27.PrevPosX, B=@n27.PrevPosX) {materialFunction=&AppendMany}.RGB
    B=@n19.RGB
   ))
  n29 = LinearInterpolate(
   A=RotateAboutAxis(NormalizedRotationAxis=CrossProduct(A=@n23, B=@n28), RotationAngle=@n20.Rigidity, PivotPoint=@n19.RGB, Position=@n17.XYZ)
   Alpha=@n25
  ) {constB=0}
  StaticSwitchParameter(True=Add(A=Multiply(A=@n15, B=@n16), B=PreviousFrameSwitch(Current Frame=@n26, Previous Frame=@n29)), False=Constant3Vector()) {defaultValue=true, parameterName=Use WPO}
  VectorParameter() {parameterName=Subsurface Color, defaultValue=<rgba 0.01 0.01 0.01 1>}
 ]
}
