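# Prototype clear-coat body material for the Accumulate player character, with a smoky emissive look over black base color, wrapped in the shared character PreFX/PostFX functions.
# Distortion-warped triplanar noise on pre-skinned position erodes a Fresnel term that indexes a color curve atlas.
# An optional vertex-color mask blends in a two-color Fresnel, and sunlight direction modulates brightness.
# A switchable, noise-distorted face mask composites through its own curve.
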Material /Game/Characters/Player/Male/Medium/Bodies/M_MED_Accumulate/Materials/M_MED_Body_PROTO
graph {
 properties {
  bUseMaterialAttributes=true
  bUsedWithMeshDeformer=true
  bUsedWithMorphTargets=true
  bUsedWithSkeletalMesh=true
  pixelDepthOffsetMode=PDOM_Legacy
  shadingModel=MSM_ClearCoat
  usedShadingModels=Clear Coat
 }
 nodes [
  Constant() {r=1}
  LinearInterpolate()
  n1 = VectorParameter() {parameterName=Color_01, defaultValue=<rgba 1 1 1 1>}
  n2 = VectorParameter() {parameterName=Color_02, defaultValue=<rgba 0 0.4993 0.535 1>}
  n3 = Power(Base=Fresnel(ExponentIn=ScalarParameter() {defaultValue=1.5, parameterName=smokeColorExponent}))
  LinearInterpolate(A=@n1.RGB, B=@n2.RGB, Alpha=@n3)
  n4 = PixelNormalWS()
  n5 = Fresnel(ExponentIn=ScalarParameter() {defaultValue=4, group=Galaxy, parameterName=Galaxy_FresEX}, Normal=@n4) {baseReflectFraction=0}
  n6 = MaterialFunctionCall() {materialFunction=&BaseFN_CharacterFunction_PreFX}
  n7 = VectorParameter() {parameterName=Scale, defaultValue=<rgba 64 64 64 1>}
  n8 = VectorParameter() {parameterName=PanSpeedAndDirection, defaultValue=<rgba 0 0 0 1>}
  n9 = Time()
  n10 = Add(
   A=Divide(A=@n6.Pre skinned Local Position, B=VectorParameter() {parameterName=DistortionScale, defaultValue=<rgba 32 32 32 1>}.RGB)
   B=Multiply(A=VectorParameter() {parameterName=DistortionPanSpeedAndDirection, defaultValue=<rgba -0.15 0.15 -0.15 1>}.RGB, B=Time())
  )
  n11 = TextureSample(UVs=ComponentMask(None=@n10) {b=true, g=true}) {texture=/Game/Characters/Player/Male/Medium/Bodies/M_MED_Viper/Textures/Glitch/T_SoftNoise}
  n12 = TextureSample(UVs=ComponentMask(None=@n10) {g=true, r=true}) {texture=&T_SoftNoise}
  n13 = PixelNormalWS()
  n14 = Abs(None=ComponentMask(None=@n13) {b=true})
  n15 = ComponentMask(None=@n10) {b=true, r=true}
  n16 = ComponentMask(None=@n13) {g=true}
  n17 = Multiply(A=LinearInterpolate(
    A=LinearInterpolate(A=@n11.RGB, B=@n12.RGB, Alpha=@n14)
    B=TextureSample(UVs=@n15) {texture=&T_SoftNoise}.RGB
    Alpha=Abs(None=@n16)
   )) {constB=2}
  n18 = Add(
   A=Add(A=Divide(A=@n6.Pre skinned Local Position, B=@n7.RGB), B=Multiply(A=@n8.RGB, B=@n9))
   B=Multiply(A=Subtract(A=@n17), B=ScalarParameter() {defaultValue=0.2, parameterName=DistortionStrength})
  )
  n19 = ComponentMask(None=@n18) {b=true, g=true}
  n20 = ComponentMask(None=@n18) {g=true, r=true}
  n21 = PixelNormalWS()
  n22 = ComponentMask(None=@n21) {b=true}
  n23 = LinearInterpolate(
   A=LinearInterpolate(
    A=TextureSample(UVs=@n19) {texture=&T_SoftNoise}.RGB
    B=TextureSample(UVs=@n20) {texture=&T_SoftNoise}.RGB
    Alpha=Abs(None=@n22)
   )
   B=TextureSample(UVs=ComponentMask(None=@n18) {b=true, r=true}) {texture=&T_SoftNoise}.RGB
   Alpha=Abs(None=ComponentMask(None=@n21) {g=true})
  )
  n24 = ChannelMaskParameter(None=@n23) {parameterName=NoiseChannel}
  n25 = ScalarParameter() {parameterName=Noise_Strength}
  n26 = Clamp(None=Subtract(A=Saturate(None=@n5), B=Multiply(A=@n24.R, B=@n25)))
  LinearInterpolate(B=@n3, Alpha=@n26)
  n27 = ComponentMask(None=@n23) {g=true, r=true}
  n28 = ScalarParameter() {defaultValue=0.035, parameterName=FaceDistortionStrength}
  n29 = TextureCoordinate()
  n30 = AppendVector(A=ScalarParameter() {parameterName=Face_X_Pos}, B=ScalarParameter() {parameterName=Face_Y_Pos})
  n31 = MaterialFunctionCall(
   UVs=Rotator(Coordinate=Add(A=@n29, B=@n30), Time=ScalarParameter() {parameterName=FaceRotation})
   Texture Scale=ScalarParameter() {defaultValue=1, parameterName=Face_UV_Scale}
  ) {materialFunction=&ScaleUVsByCenter}
  n32 = TextureSampleParameter2D(UVs=Add(A=Multiply(A=@n27, B=@n28), B=@n31.UVs)) {
   parameterName=FaceMask
   samplerType=SAMPLERTYPE_Masks
   texture=/Game/Accessories/FORT_Backpacks/Backpack_M_MED_Accumulate/Textures/T_MED_Accumulate_Backpack_FX
  }
  n33 = Multiply(A=ChannelMaskParameter(None=@n32.RGB) {parameterName=FaceMask}.R, B=@n31.0-1 mask)
  LinearInterpolate(
   A=VectorParameter() {parameterName=Face_Color_02, defaultValue=<rgba 0 0.9333 1 1>}.RGB
   B=VectorParameter() {parameterName=Face_Color_01, defaultValue=<rgba 1 1 1 1>}.RGB
   Alpha=@n33
  )
  n34 = Constant()
  n35 = CurveAtlasRowParameter(CurveTime=@n33) {
   atlas=&AthenaCurveAtlas01
   curve=&/AccumulateBlackColorCurve
   defaultValue=276
   parameterName=FaceColorCurve
  }
  n36 = CurveAtlasRowParameter(CurveTime=@n26) {
   atlas=&AthenaCurveAtlas01
   curve=&AccumulateBlackColorCurve
   defaultValue=276
   parameterName=BG_ColorCurve
  }
  n37 = Fresnel(ExponentIn=ScalarParameter() {defaultValue=2, parameterName=Exponent}, Normal=@n4)
  n38 = VertexColor()
  n39 = OneMinus(None=@n38.G)
  n40 = Multiply(
   A=LinearInterpolate(A=@n38.G, B=Clamp(None=@n39), Alpha=ScalarParameter() {parameterName=FlipVertColorMask})
   B=ScalarParameter() {group=Accumulate, parameterName=UsevertColorMask}
  )
  n41 = LinearInterpolate(
   A=Multiply(A=@n36, B=ScalarParameter() {defaultValue=0.25, parameterName=Flame_Brightness})
   B=LinearInterpolate(A=@n2.RGB, B=@n1.RGB, Alpha=Saturate(None=@n37))
   Alpha=@n40
  )
  n42 = CollectionParameter() {
   collection=&FortniteMaterialParameters
   parameterName=SunlightDirectionalVector
  }
  n43 = ComponentMask(None=@n42) {b=true, g=true, r=true}
  n44 = PixelNormalWS()
  n45 = LinearInterpolate(
   A=Multiply(A=@n41, B=ScalarParameter() {defaultValue=0.75, parameterName=Rim_ShadowBrightness})
   B=Multiply(A=@n41, B=ScalarParameter() {defaultValue=1.5, parameterName=Rim_LitBrightness})
   Alpha=Clamp(None=DotProduct(A=@n43, B=@n44))
  )
  n46 = LinearInterpolate(A=Add(A=@n35, B=@n45), B=Multiply(A=@n35, B=@n45), Alpha=ScalarParameter() {parameterName=MultiplyFace})
  n47 = ScalarParameter() {parameterName=UseFaceAndWhite}
  n48 = LinearInterpolate(
   A=VectorParameter() {parameterName=SS_StencilValue, defaultValue=<rgba 0 0.004 0 1>}.RGB
   B=VectorParameter() {parameterName=SS_StencilValue2, defaultValue=<rgba 0 0.004 0 1>}.RGB
   Alpha=@n40
  )
  n49 = SetMaterialAttributes!(
   MaterialAttributes=GetMaterialAttributes(None=@n6.Result).MaterialAttributes
   Base Color=Multiply(
    A=TextureSample() {samplerType=SAMPLERTYPE_Masks, texture=/Game/Characters/Player/Male/Medium/Bodies/M_MED_Accumulate/Textures/T_HeadSprites_TEMP}.RGB
    B=@n34
   )
   Metallic=@n34
   Specular=@n34
   Emissive Color=StaticSwitchParameter(True=LinearInterpolate(A=@n35, B=@n46, Alpha=@n47), False=@n45) {parameterName=UseFace}
   Opacity=Constant() {r=1}
   Normal=TextureSampleParameter2D() {parameterName=Normal, samplerType=SAMPLERTYPE_Normal, texture=&FlatNormalBlue}.RGB
   Subsurface Color=@n48
   Roughness=@n34
   Clear Coat=ComponentMask(None=@n48) {r=true}
   Clear Coat Roughness=ComponentMask(None=@n48) {g=true}
  ) {attributeSetTypes [
    #1
    #2
    #3
    #4
    #5
    #6
    #7
    #8
    #9
    #10
   ]}
  MaterialFunctionCall(AO=@n6.AO, Distance Blend=@n6.AO, Fresnel=@n6.AO, Gmap=@n6.AO, Pre Skinned Local Position=@n6.AO, Material=@n49) {materialFunction=&BaseFN_CharacterFunction_PostFX}
  Multiply(
   A=VectorParameter() {parameterName=BG_Color, defaultValue=<rgba 0 0 0 1>}.RGB
   B=ScalarParameter() {defaultValue=1, parameterName=BG_ColorBrightness}
  )
  n50 = PreSkinnedPosition()
  ShaderStageSwitch(PixelShader=VertexInterpolator(VS=@n50).PS, VertexShader=@n50)
  VectorParameter() {parameterName=FaceColor, defaultValue=<rgba 1 1 1 1>}
  VectorParameter() {parameterName=HeadSocketLocation, defaultValue=<rgba 0 0 0 1>}
 ]
}
